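# Niagara effect for the Smoke Bomb Fail emote: a brief warm spark pop followed by puffs of grey smoke.
# A short-lived core, light flash, GPU streaks, sparks and gravity-pulled drips form the pop, with no collision modules despite the BouncingSparks name.
# Three smoke emitters push sprites outward and upward from flat cylinders under drag, using SubUV animation, fading alpha and curve-driven dynamic material parameters.
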NiagaraSystem /BRCosmetics/Effects/Fort_Effects/Emotes/SmokeBombFail/NS_Emote_SmokeBombFail
dependencies {
 usedDataInterfaces [
  &NiagaraDataInterfaceCurve
  &NiagaraDataInterfaceVector2DCurve
  &NiagaraDataInterfaceVectorCurve
  &NiagaraDataInterfaceSpriteRendererInfo
 ]
 usedDynamicInputs []
 usedModules [&SystemState]
}
emitters {
 BouncingSparks {
  emitterUpdateScript [
   &EmitterState(Life Cycle Mode=::NewEnumerator1, Loop Behavior=::NewEnumerator1, Loop Duration=0.2, UseLoopDelay=false)
   &SpawnRate(SpawnRate=200, Use Spawn Probability=false) {module=SpawnRate001}
  ]
  particleSpawnScript [
   &InitializeParticle(
    Lifetime=&RandomRangeFloat(Minimum=0.025, Maximum=0.1)
    Sprite Rotation Angle=&RandomRangeFloat(Minimum=0, Maximum=1)
    Sprite Rotation Mode=::NewEnumerator1
    Sprite Size=&Vector2DFromFloat(Value=&RandomRangeFloat(Minimum=4, Maximum=20))
    Sprite Size Mode=::NewEnumerator4
    UsePositionOffset=false
    Write Material Random=true
   )
   &Color(Color=&MakeLinearColorFromVectorAndFloat("Vector (RGB)"=<1 0.8667 0.579>, "Float (Alpha)"=&ClampFloat(Float=0.5)))
   &ShapeLocation(
    MaskSpawnGroup=false
    Override Seed=false
    Sphere Radius=&RandomRangeFloat(Minimum=10, Maximum=20)
    UseOwnerScale=true
   )
   &AddVelocity(Velocity=&Multiply_VectorByFloat(Vector=$Particles.ShapeLocation.ShapeVector, Float=2))
  ]
  particleUpdateScript [
   &ParticleState()
   &ScaleSpriteSize(
    Scale Factor=&Vector2DFromFloat(Value=&FloatFromCurve(FloatCurve=DataInterface() {propertyValues {Curve {keys [<at 0 0>,<at 1 1>]}}}))
    Scale Sprite Size Mode=::NewEnumerator2
   )
   /Niagara/Modules/Update/SubUV/SubUVAnimation(Number Of Frames=3)
   &Drag(Drag=3, Use Linear Drag=true, Use Rotational Drag=true)
   &SolveForcesAndVelocity(Clamp Velocity=false, Limit Acceleration=false)
  ]
  rendererClasses [&NiagaraSpriteRendererProperties]
  renderers [{rendererClass=&NiagaraSpriteRendererProperties}]
 }
 Core {
  emitterUpdateScript [
   &EmitterState()
   &SpawnBurst_Instantaneous(Spawn Count=5, Use Spawn Probability=false, UseLoopCountLimit=false)
  ]
  particleSpawnScript [
   &InitializeParticle(
    Lifetime=&RandomRangeFloat(Minimum=0.025, Maximum=0.1)
    Sprite Rotation Angle=&RandomRangeFloat(Minimum=0, Maximum=1)
    Sprite Rotation Mode=::NewEnumerator1
    Sprite Size=&Vector2DFromFloat(Value=&RandomRangeFloat(Minimum=10, Maximum=40))
    Sprite Size Mode=::NewEnumerator4
    UsePositionOffset=false
    Write Material Random=true
   )
   &Color(Color=&MakeLinearColorFromVectorAndFloat("Vector (RGB)"=<3 1.5575 0.297>, "Float (Alpha)"=&ClampFloat(Float=1)))
  ]
  particleUpdateScript [
   &ParticleState()
   &ScaleSpriteSize(
    Scale Sprite Size Mode=::NewEnumerator0
    Uniform Scale Factor=&FloatFromCurve(FloatCurve=DataInterface() {propertyValues {Curve {keys [<at 0 0>,<at 1 1>]}}})
   )
   &SubUVAnimation(Number Of Frames=3)
   &CameraOffset(Camera Offset Amount=2)
  ]
  rendererClasses [&NiagaraSpriteRendererProperties]
  renderers [{rendererClass=&NiagaraSpriteRendererProperties}]
 }
 LightFlash {
  emitterUpdateScript [
   &EmitterState()
   &SpawnBurst_Instantaneous(Use Spawn Probability=false, UseLoopCountLimit=false)
  ]
  particleSpawnScript [
   &InitializeParticle(Color=<rgba 5 1.18 0.98 1>, Lifetime=0.1, UsePositionOffset=false, Write Material Random=true)
   /SetVariables(Particles.LightRadius=400) {isSetParameters=true}
  ]
  particleUpdateScript [&ParticleState()]
  rendererClasses [&NiagaraLightRendererProperties]
  renderers [{rendererClass=&NiagaraLightRendererProperties, rendererIndex=0}]
 }
 LongLived_drips {
  emitterUpdateScript [
   &EmitterState(Life Cycle Mode=::NewEnumerator1, Loop Behavior=::NewEnumerator1, Loop Duration=0.2, UseLoopDelay=false)
   &SpawnRate(SpawnRate=150, Use Spawn Probability=false) {module=SpawnRate001}
  ]
  particleSpawnScript [
   &InitializeParticle(
    Lifetime=&RandomRangeFloat(Minimum=0.1, Maximum=0.5)
    Sprite Size=&RandomRangeVector2D(Minimum=<2 2.5>, Maximum=<4 3>)
    Sprite Size Mode=::NewEnumerator4
    UsePositionOffset=false
    Write Material Random=true
   )
   &AddVelocity(Velocity=&RandomRangeVector(Minimum=<-750 -750 -10>, Maximum=<750 750 500>))
   &Color(Color=&MakeLinearColorFromVectorAndFloat("Vector (RGB)"=<3 1.5575 0.297>, "Float (Alpha)"=&ClampFloat(Float=1)))
  ]
  particleUpdateScript [
   &ParticleState()
   &ScaleSpriteSize(
    Scale Factor=&Vector2DFromCurve(
     Vector2Curve=DataInterface() {propertyValues {
       XCurve {keys [<at 0 1 RCIM_Cubic>, {interpMode=RCIM_Cubic, leaveTangent=-1.9252, time=1, value=0.2}]}
       YCurve {keys [<at 0 3 RCIM_Cubic>,<at 1 0 RCIM_Cubic>]}
      }}
     ScaleCurve=<1 1>
    )
    Scale Sprite Size Mode=::NewEnumerator2
   )
   &AccelerationForce(Acceleration=<0 0 -1000>)
   &Drag(Drag=3, Use Linear Drag=true, Use Rotational Drag=true)
   &SolveForcesAndVelocity(Clamp Velocity=false, Limit Acceleration=false)
  ]
  rendererClasses [&NiagaraSpriteRendererProperties]
  renderers [{rendererClass=&NiagaraSpriteRendererProperties}]
 }
 SmokeLinger {
  emitterUpdateScript [
   &EmitterState(Life Cycle Mode=::NewEnumerator1, Loop Behavior=::NewEnumerator1, Loop Duration=0.3, UseLoopDelay=true)
   &SpawnBurst_Instantaneous(Spawn Count=3, Use Spawn Probability=false, UseLoopCountLimit=false)
   &SpawnRate(SpawnRate=50, Use Spawn Probability=false)
  ]
  particleSpawnScript [
   &InitializeParticle(
    Color Maximum=<rgba 0.35 0.35 0.35 2>
    Color Minimum=<rgba 0.15 0.15 0.15 1>
    Color Mode=::NewEnumerator2
    Lifetime=&RandomRangeFloat(Minimum=0.5, Maximum=0.75)
    Sprite Rotation Mode=::NewEnumerator2
    Sprite Size=&Vector2DFromFloat(Value=&RandomRangeFloat(Minimum=50, Maximum=125))
    Sprite Size Mode=::NewEnumerator4
    UsePositionOffset=false
    Write Material Random=true
   )
   &ShapeLocation(
    Cylinder Height=1
    Cylinder Radius=50
    MaskSpawnGroup=false
    Override Seed=false
    Rotation Axis=<0 0 1>
    Rotation Mode=::NewEnumerator1
    Shape Primitive=::NewEnumerator1
    Surface Only=false
    UseOwnerScale=true
   )
   &AddVelocity(Velocity=&Multiply_VectorByFloat(
     Vector=&Multiply_Vector(A=$Particles.ShapeLocation.ShapeVector, B=<1 1 0>)
     Float=&RandomRangeFloat(Minimum=5, Maximum=2)
    ))
   &AddVelocity(Velocity=&RandomRangeVector(Minimum=<0 0 0>, Maximum=<0 0 200>)) {module=AddVelocity001}
   &CameraOffset(Camera Offset Amount=18)
  ]
  particleUpdateScript [
   &ParticleState()
   &ScaleColor(
    Scale Alpha=&FloatFromCurve(FloatCurve=DataInterface() {propertyValues {Curve {keys [
         {arriveTangent=0.616, interpMode=RCIM_Cubic, time=0, value=0}
         {arriveTangent=0.0042, interpMode=RCIM_Cubic, leaveTangent=0.0042, time=0.2235, value=1.4641}
         {interpMode=RCIM_Cubic, leaveTangent=-0.0493, time=0.9453, value=0.004}
        ]}}})
    Scale RGB=&VectorFromCurve(VectorCurve=DataInterface() {propertyValues {
       XCurve {keys [<at 0.1007 0.68 RCIM_Cubic>,<at 1 0.084 RCIM_Cubic>]}
       YCurve {keys [<at 0.1007 0.68 RCIM_Cubic>,<at 1 0.0887 RCIM_Cubic>]}
       ZCurve {keys [<at 0.1007 0.68 RCIM_Cubic>,<at 1 0.099 RCIM_Cubic>]}
      }})
    ScaleA=true
    ScaleRGB=true
   )
   /Niagara/Modules/Update/SubUV/V2/SubUVAnimation(
    End Frame Range Override=46
    "Sprite Renderer "=DataInterface() {propertyValues {SpriteRenderer=&self:SmokeLinger_0.NiagaraSpriteRendererProperties_0}}
    Start Frame Range Override=20
    SubUV Animation Mode:ENiagara_SubUVLookupModeV2=::NewEnumerator1
    UseEndFrame=true
    UseStartFrame=true
   ) {module=SubUVAnimation001}
   &DynamicMaterialParameters(
    Index 0 Param 1=&RandomRangeFloat(Minimum=0, Maximum=1)
    Index 0 Param 2=1
    Index 0 Param 3=&FloatFromCurve(FloatCurve=DataInterface() {propertyValues {Curve {keys [
         <at 0.5 1 RCIM_Cubic>
         {arriveTangent=-3.3333, interpMode=RCIM_Cubic, leaveTangent=-3.3333, time=0.6598, value=0.5089}
         <at 0.8 0 RCIM_Cubic>
        ]}}})
    Index 0 Param 4=0.5
    Module.Index0Param0Write=true
    Module.Index0Param1Write=true
    Module.Index0Param2Write=true
    Module.Index0Param3Write=true
   ) {module=DynamicMaterialParameters001}
   &Drag(Drag=3, Use Linear Drag=true, Use Rotational Drag=true)
   &SolveForcesAndVelocity(Clamp Velocity=false, Limit Acceleration=false)
  ]
  rendererClasses [&NiagaraSpriteRendererProperties]
  renderers [{rendererClass=&NiagaraSpriteRendererProperties}]
 }
 SmokeLingerUp {
  emitterUpdateScript [
   &EmitterState(Life Cycle Mode=::NewEnumerator1, Loop Behavior=::NewEnumerator1, Loop Duration=1.5, UseLoopDelay=true)
   &SpawnBurst_Instantaneous(Spawn Count=10, Use Spawn Probability=false, UseLoopCountLimit=false)
  ]
  particleSpawnScript [
   &InitializeParticle(
    Color Maximum=<rgba 0.53 0.53 0.53 3>
    Color Minimum=<rgba 0.38 0.38 0.38 1>
    Color Mode=::NewEnumerator2
    Lifetime=&RandomRangeFloat(Minimum=1.5, Maximum=2)
    Sprite Rotation Angle=&RandomRangeFloat(Minimum=0, Maximum=1)
    Sprite Rotation Mode=::NewEnumerator1
    Sprite Size=&Vector2DFromFloat(Value=&RandomRangeFloat(Minimum=50, Maximum=125))
    Sprite Size Mode=::NewEnumerator4
    UsePositionOffset=false
    Write Material Random=true
   )
   &ShapeLocation(
    Cylinder Height=1
    Cylinder Radius=2.5
    MaskSpawnGroup=false
    Override Seed=false
    Rotation Axis=<0 0 1>
    Rotation Mode=::NewEnumerator1
    Shape Primitive=::NewEnumerator1
    Surface Only=false
    UseOwnerScale=true
   )
   &AddVelocity(Velocity=&Multiply_VectorByFloat(
     Vector=&Multiply_Vector(A=$Particles.ShapeLocation.ShapeVector, B=<1 1 0>)
     Float=&RandomRangeFloat(Minimum=60, Maximum=30)
    ))
   &AddVelocity(Velocity=&RandomRangeVector(Minimum=<0 0 100>, Maximum=<0 0 300>)) {module=AddVelocity001}
   &CameraOffset(Camera Offset Amount=18)
  ]
  particleUpdateScript [
   &ParticleState()
   &ScaleSpriteSize(
    Scale Sprite Size Mode=::NewEnumerator0
    Uniform Scale Factor=&FloatFromCurve(FloatCurve=DataInterface() {propertyValues {Curve {keys [{arriveTangent=2.6074, interpMode=RCIM_Cubic, leaveTangent=2.6074, tangentMode=RCTM_User, time=0, value=0}, <at 1 1>]}}})
   )
   &ScaleColor(
    Scale Alpha=&FloatFromCurve(FloatCurve=DataInterface() {propertyValues {Curve {keys [
         {arriveTangent=0.616, interpMode=RCIM_Cubic, time=0, value=0}
         {arriveTangent=0.0042, interpMode=RCIM_Cubic, leaveTangent=0.0042, time=0.2235, value=1.4641}
         {interpMode=RCIM_Cubic, leaveTangent=-0.0493, time=0.9453, value=0.004}
        ]}}})
    Scale RGB=&VectorFromCurve(VectorCurve=DataInterface() {propertyValues {
       XCurve {keys [<at 0.1007 0.68 RCIM_Cubic>,<at 1 0.084 RCIM_Cubic>]}
       YCurve {keys [<at 0.1007 0.68 RCIM_Cubic>,<at 1 0.0887 RCIM_Cubic>]}
       ZCurve {keys [<at 0.1007 0.68 RCIM_Cubic>,<at 1 0.099 RCIM_Cubic>]}
      }})
    ScaleA=true
    ScaleRGB=true
   )
   &SubUVAnimation(
    End Frame Range Override=46
    "Sprite Renderer "=DataInterface() {propertyValues {SpriteRenderer=&self:SmokeLingerUp_3.NiagaraSpriteRendererProperties_3}}
    SubUV Animation Mode:ENiagara_SubUVLookupModeV2=::NewEnumerator1
    SubUV Curve=DataInterface() {propertyValues {Curve {keys [
        {arriveTangent=4.1785, interpMode=RCIM_Cubic, leaveTangent=4.1785, tangentMode=RCTM_User, time=0, value=0}
        {arriveTangent=1.1449, interpMode=RCIM_Cubic, leaveTangent=1.1449, tangentMode=RCTM_User, time=0.2, value=0.4348}
        {arriveTangent=0.2086, interpMode=RCIM_Cubic, leaveTangent=0.2086, tangentMode=RCTM_User, time=1, value=1}
       ]}}}
    UseEndFrame=true
    UseStartFrame=true
   ) {module=SubUVAnimation001}
   &DynamicMaterialParameters(
    Index 0 Param 1=&RandomRangeFloat(Minimum=0, Maximum=1)
    Index 0 Param 2=1
    Index 0 Param 3=&FloatFromCurve(FloatCurve=DataInterface() {propertyValues {Curve {keys [
         <at 0.5 1 RCIM_Cubic>
         {arriveTangent=-3.3333, interpMode=RCIM_Cubic, leaveTangent=-3.3333, time=0.6598, value=0.5089}
         <at 0.8 0 RCIM_Cubic>
        ]}}})
    Index 0 Param 4=0.5
    Module.Index0Param0Write=true
    Module.Index0Param1Write=true
    Module.Index0Param2Write=true
    Module.Index0Param3Write=true
   ) {module=DynamicMaterialParameters001}
   &Drag(Drag=2.5, Use Linear Drag=true, Use Rotational Drag=true)
   &SolveForcesAndVelocity(Clamp Velocity=false, Limit Acceleration=false)
  ]
  rendererClasses [&NiagaraSpriteRendererProperties]
  renderers [{rendererClass=&NiagaraSpriteRendererProperties}]
 }
 SmokeRoundUp {
  emitterUpdateScript [
   &EmitterState(Life Cycle Mode=::NewEnumerator1, Loop Behavior=::NewEnumerator1, Loop Duration=2.6, UseLoopDelay=true)
   &SpawnBurst_Instantaneous(Spawn Count=10, Use Spawn Probability=false, UseLoopCountLimit=false)
   &SpawnRate(SpawnRate=10, Use Spawn Probability=false)
  ]
  particleSpawnScript [
   &InitializeParticle(
    Lifetime=&RandomRangeFloat(Minimum=1.5, Maximum=2)
    Sprite Size=&Vector2DFromFloat(Value=&RandomRangeFloat(Minimum=100, Maximum=200))
    Sprite Size Mode=::NewEnumerator4
    UsePositionOffset=false
    Write Material Random=true
   )
   &CameraOffset(Camera Offset Amount=&RandomRangeFloat(Minimum=40, Maximum=50))
   &Color(Color=&MakeLinearColorFromVectorAndFloat(
     "Vector (RGB)"=<0.8 0.8 0.8>
     "Float (Alpha)"=&ClampFloat(Float=&RandomRangeFloat(Minimum=0.2, Maximum=0.3))
    ))
   &ShapeLocation(
    Cylinder Height=1
    Cylinder Radius=2.5
    MaskSpawnGroup=false
    Override Seed=false
    Rotation Axis=<0 0 1>
    Rotation Mode=::NewEnumerator1
    Shape Primitive=::NewEnumerator1
    Surface Only=false
    UseOwnerScale=true
   )
   &AddVelocity(Velocity=&Multiply_VectorByFloat(
     Vector=&Multiply_Vector(A=$Particles.ShapeLocation.ShapeVector, B=<1 1 0>)
     Float=&RandomRangeFloat(Minimum=60, Maximum=30)
    ))
   &AddVelocity(Velocity=&RandomRangeVector(Minimum=<0 0 100>, Maximum=<0 0 300>)) {module=AddVelocity001}
  ]
  particleUpdateScript [
   &ParticleState()
   &ScaleSpriteSize(
    Scale Sprite Size Mode=::NewEnumerator0
    Uniform Scale Factor=&FloatFromCurve(FloatCurve=DataInterface() {propertyValues {Curve {keys [{arriveTangent=0.7079, time=-0.0011, value=0.0573}, {interpMode=RCIM_Cubic, leaveTangent=0.0935, time=1.0126, value=0.7652}]}}})
   )
   &DynamicMaterialParameters(
    Index 0 Param 1=&FloatFromCurve(FloatCurve=DataInterface() {propertyValues {Curve {keys [<at 0.5 0.5 RCIM_Cubic>,<at 1 -0.1>]}}})
    Index 0 Param 4=1
    Module.Index0Param0Write=true
    Module.Index0Param1Write=true
    Module.Index0Param2Write=true
    Module.Index0Param3Write=true
   )
   &ScaleColor(
    Scale Alpha=&FloatFromCurve(FloatCurve=DataInterface() {propertyValues {Curve {keys [<at 0 0>,<at 0.25 1>,<at 1 0>]}}})
    Scale RGB=&VectorFromCurve(VectorCurve=DataInterface() {propertyValues {XCurve {keys [<at 0 1>,<at 1 1>]}, YCurve {keys [<at 0 1>,<at 1 1>]}, ZCurve {keys [<at 0 1>,<at 1 1>]}}})
    ScaleA=true
    ScaleRGB=true
   )
   &SubUVAnimation(
    Random Start Frame=true
    "Sprite Renderer "=DataInterface() {propertyValues {SpriteRenderer=&self:SmokeRoundUp_2.NiagaraSpriteRendererProperties_2}}
    SubUV Lookup Index:NiagaraFloat=$Particles.Age
    SubUV Lookup Index Scale=0.25
    UseEndFrame=false
    UseStartFrame=false
   ) {module=SubUVAnimation001}
   &Drag(Drag=2.5, Use Linear Drag=true, Use Rotational Drag=true)
   &SolveForcesAndVelocity(Clamp Velocity=false, Limit Acceleration=false)
  ]
  rendererClasses [&NiagaraSpriteRendererProperties]
  renderers [{rendererClass=&NiagaraSpriteRendererProperties}]
 }
 Streaks {
  simTarget=GPUComputeSim
  emitterUpdateScript [
   &EmitterState()
   &SpawnBurst_Instantaneous(Spawn Count=50, Use Spawn Probability=false, UseLoopCountLimit=false)
  ]
  particleSpawnScript [
   &InitializeParticle(
    Lifetime=&RandomRangeFloat(Minimum=0.025, Maximum=0.05)
    Sprite Size=&RandomRangeVector2D(Minimum=<2 6>, Maximum=<3 8>)
    Sprite Size Mode=::NewEnumerator4
    UsePositionOffset=false
    Write Material Random=true
   )
   &AddVelocity(Velocity=&RandomRangeVector(Minimum=<-250 -250 -250>, Maximum=<250 250 250>))
   &Color(Color=&MakeLinearColorFromVectorAndFloat("Vector (RGB)"=<3 1.5575 0.297>, "Float (Alpha)"=&ClampFloat(Float=1)))
  ]
  particleUpdateScript [
   &ParticleState()
   &ScaleSpriteSize(
    Scale Factor=&Vector2DFromCurve(
     Vector2Curve=DataInterface() {propertyValues {
       XCurve {keys [<at 0 1 RCIM_Cubic>, {interpMode=RCIM_Cubic, leaveTangent=-1.9252, time=1, value=0.2}]}
       YCurve {keys [<at 0 3 RCIM_Cubic>,<at 1 0 RCIM_Cubic>]}
      }}
     ScaleCurve=<1 1>
    )
    Scale Sprite Size Mode=::NewEnumerator2
   )
   &AccelerationForce(Acceleration=<0 0 -980>)
   &Drag(Use Linear Drag=true, Use Rotational Drag=true)
   &SolveForcesAndVelocity(Clamp Velocity=false, Limit Acceleration=false)
  ]
  rendererClasses [&NiagaraSpriteRendererProperties]
  renderers [{rendererClass=&NiagaraSpriteRendererProperties}]
 }
}
summary {
 systemName=NS_Emote_SmokeBombFail
 emitters [
  {emitterName=Streaks, simTarget=GPUComputeSim, rendererClasses [&NiagaraSpriteRendererProperties]}
  {emitterName=LongLived_drips, rendererClasses [&NiagaraSpriteRendererProperties]}
  {emitterName=BouncingSparks, rendererClasses [&NiagaraSpriteRendererProperties]}
  {emitterName=Core, rendererClasses [&NiagaraSpriteRendererProperties]}
  {emitterName=LightFlash, rendererClasses [&NiagaraLightRendererProperties]}
  {emitterName=SmokeLinger, rendererClasses [&NiagaraSpriteRendererProperties]}
  {emitterName=SmokeRoundUp, rendererClasses [&NiagaraSpriteRendererProperties]}
  {emitterName=SmokeLingerUp, rendererClasses [&NiagaraSpriteRendererProperties]}
 ]
 userVariables []
}
